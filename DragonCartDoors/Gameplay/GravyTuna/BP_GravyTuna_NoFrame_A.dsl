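; Frameless GravyTuna door Blueprint that swings and shifts its pivot open and closed with garage door sounds.
; TL_DoorOpen drives DoorOperationPercentage and MoveDoorByPercentage lerps pivot roll, Y and height, while authority sets door state on finish.
; It forces health from a scalable float, and destroys itself when its linked actor dies or a looping timer finds its child actor invalid.
; KillBox, AutoReset and LaunchPlayers events are empty, though a launch-velocity helper remains.
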(event EventBeginPlay
  (|CallParentFunction)
  (Collision|SetCollisionEnabled (Variables|Default|GetBlockingBox) "NoCollision")
  (Utilities|Time|SetTimerbyEvent (delegate Custom|DestroyAfterChildDeath) 3e+01 true true 45.0 15.0))

(event Custom|DestroyAfterChildDeath
  (Utilities|IsValid (Variables|Default|GetChildActor)
    (:Is%20Not%20Valid
      (Rendering|SetActorHiddenInGame true)
      (Actor|DestroyActor))))

(event EventBP_OnActive
  (params (EnableAlternateActive :type Boolean) (EnableOpenCues :type Boolean))
  (|CallParentFunction EnableAlternateActive EnableOpenCues)
  (Audio|PlaySoundatLocation "/Game/Sounds/Fort_Doors/World/Door_Garage01_Open_Cue.Door_Garage01_Open_Cue" (Transformation|GetActorLocation))
  (bind (_direction _rotatetrack) (Timeline|TL_DoorOpen PlayFromStart)
    (:Update
      (Variables|Default|SetDoorOperationPercentage _rotatetrack))
    (:Finished
      (switch Utilities|FlowControl|Switch|SwitchonETimelineDirection _direction
        (:Forward
          (Utilities|FlowControl|Sequence
            (:then_1
              (Networking|SwitchHasAuthority
                (:Authority
                  (Variables|Default|SetDoorState "NewEnumerator2")
                  (Default|CallOnDoorFinishActivation))))))
        (:Backward
          (Networking|SwitchHasAuthority
            (:Authority
              (Variables|Default|SetDoorState "NewEnumerator0")
              (Default|CallOnDoorFinishReset))))))))

(event Default|EventBP_OnReset
  (params (EnableAlternateReset :type Boolean) (EnableOpenCues :type Boolean))
  (|CallParentFunction EnableAlternateReset EnableOpenCues)
  (Audio|PlaySoundatLocation "/Game/Sounds/Fort_Doors/World/Door_Garage01_Close_Cue.Door_Garage01_Close_Cue" (Transformation|GetActorLocation))
  (Timeline|TL_DoorOpen ReverseFromEnd))

(event Custom|AutoReset)

(event Custom|LaunchPlayers)

(event OnComponentBeginOverlap(KillBox)
  (params (OverlappedComponent :type Primitive%20Component%20Object%20Reference) (OtherActor :type Actor%20Object%20Reference) (OtherComp :type Primitive%20Component%20Object%20Reference) (OtherBodyIndex :type Integer) (bFromSweep :type Boolean) (SweepResult :type Hit%20Result%20Structure%20%28by%20ref%29)))

(event OnComponentEndOverlap(KillBox)
  (params (OverlappedComponent :type Primitive%20Component%20Object%20Reference) (OtherActor :type Actor%20Object%20Reference) (OtherComp :type Primitive%20Component%20Object%20Reference) (OtherBodyIndex :type Integer)))

(event OnLinkedActorDestroyed(FortLinkToActor)
  (params (DamageCauser :type Actor%20Object%20Reference))
  (Rendering|SetActorHiddenInGame true)
  (Actor|DestroyActor))

(event EventSetDoorHealthFromDataTable
  (Gameplay|ForceBuildingHealth self (Math|Conversions|ToFloat(Integer) (Math|Float|Truncate (ScalableFloat|GetValueatLevel (Variables|Default|GetScalarDoorHealth))))))

(fn DetermineLaunchDirection
  (params (Actor :type Actor%20Object%20Reference))
  (Variables|Default|SetLaunchVelocity (* (* (+ (Transformation|GetActorRightVector) (Math|Vector|MakeVector 0.0 0.0 (Variables|Default|GetLaunchBoostInitialZ))) (Variables|Default|GetLaunchPower)) (Math|Vector|MakeVector (* (.x (* (+ (Transformation|GetActorRightVector) (Math|Vector|MakeVector 0.0 0.0 (Variables|Default|GetLaunchBoostInitialZ))) (Variables|Default|GetLaunchPower))) (Variables|Default|GetLaunchBoostX)) (* (.y (* (+ (Transformation|GetActorRightVector) (Math|Vector|MakeVector 0.0 0.0 (Variables|Default|GetLaunchBoostInitialZ))) (Variables|Default|GetLaunchPower))) (Variables|Default|GetLaunchBoostY)) (* (.z (* (+ (Transformation|GetActorRightVector) (Math|Vector|MakeVector 0.0 0.0 (Variables|Default|GetLaunchBoostInitialZ))) (Variables|Default|GetLaunchPower))) (Variables|Default|GetLaunchBoostZ)))))
  (return))

(fn MoveDoorByPercentage
  (params (Percentage :type Float%20%28double-precision%29))
  (|CallParentFunction Percentage)
  (Transformation|SetRelativeRotation (Variables|Default|GetDoorPivot) (Math|Rotator|MakeRotator (Math|Float|Lerp (Variables|Default|GetRotationStartX) (Variables|Default|GetRotationEndX) Percentage) (.pitch (Variables|/Script/Engine.SceneComponent|GetRelativeRotation (Variables|Default|GetDoorPivot))) (.yaw (Variables|/Script/Engine.SceneComponent|GetRelativeRotation (Variables|Default|GetDoorPivot)))) false false)
  (Transformation|SetRelativeLocation (Variables|Default|GetDoorPivot) (Math|Vector|MakeVector (.x (Variables|/Script/Engine.SceneComponent|GetRelativeLocation (Variables|Default|GetDoorPivot))) (Math|Float|Lerp (Variables|Default|GetRotationStartY) (Variables|Default|GetRotationEndY) Percentage) (Math|Float|Lerp (Variables|Default|GetHeightStartX) (Variables|Default|GetHeightEndX) Percentage)) false false))

(fn OnRep_SafteyLightOn
  (if (Variables|Default|GetSafteyLightOn)
    (Rendering|Material|SetMaterial (Variables|Default|GetSafteyLight) 0 "/DragonCartDoors/Assets/MI_FlatColorParent_Red.MI_FlatColorParent_Red")
    (return)
    (else
      (Rendering|Material|SetMaterial (Variables|Default|GetSafteyLight) 0 "/DragonCartDoors/Assets/MI_FlatColorParent_Iron.MI_FlatColorParent_Iron")
      (return))))

(fn ConstructionScript
  (|CallParentFunction)
  (Transformation|SetRelativeScale3D (Variables|Default|GetChildActor) (Variables|Default|GetDoorScaleOverride)))
