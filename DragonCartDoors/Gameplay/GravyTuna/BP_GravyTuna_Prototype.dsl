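; Prototype launcher door (GravyTuna) that swings on a timeline and flings players and vehicles in its KillBox.
; Activation plays the timeline, driving door percentage and panel collision; authority sets state at each end, and reset reverses it.
; A forward-only timeline event launches overlapping pawns, subject to a tag check, and vehicles along the door's boosted right vector.
; KillBox overlaps toggle a replicated safety light between red and iron materials.
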(event Default|EventBP_OnReset
  (params (EnableAlternateReset :type Boolean) (EnableOpenCues :type Boolean))
  (|CallParentFunction EnableAlternateReset false)
  ;[1 1 1] On Activate Door
  (bind (_direction _rotatetrack) (Timeline|TL_DoorOpen ReverseFromEnd)
    (:Update
      (Variables|Default|SetDoorOperationPercentage _rotatetrack))
    (:Finished
      (switch Utilities|FlowControl|Switch|SwitchonETimelineDirection _direction
        (:Forward
          (Collision|SetCollisionEnabled (Variables|Default|GetPanel) "QueryOnly")
          (Networking|SwitchHasAuthority
            (:Authority
              (Variables|Default|SetDoorState "NewEnumerator2")
              (Default|CallOnDoorFinishActivation))))
        (:Backward
          (Collision|SetCollisionEnabled (Variables|Default|GetPanel) "QueryAndPhysics")
          (Networking|SwitchHasAuthority
            (:Authority
              (Variables|Default|SetDoorState "NewEnumerator0")
              (Default|CallOnDoorFinishReset))))))
    (:LaunchPlayer
      (switch Utilities|FlowControl|Switch|SwitchonETimelineDirection _direction
        (:Forward
          (CallFunction|LaunchPlayers))))))

(event EventBP_OnActive
  (params (EnableAlternateActive :type Boolean) (EnableOpenCues :type Boolean))
  ;[1 1 1] On Activate Door
  (|CallParentFunction EnableAlternateActive false)
  (Collision|SetCollisionEnabled (Variables|Default|GetPanel) "NoCollision")
  (Timeline|TL_DoorOpen PlayFromStart))

(event Custom|LaunchPlayers
  ;[1 1 1] LaunchPlayers
  (for _array_element (Collision|GetOverlappingActors (Variables|Default|GetKillBox) 0 :!class /Script/Engine.PrimitiveComponent)
    (bind _asfort_player_pawn (Utilities|Casting|CastToFortPlayerPawn _array_element)
      (:then
        (CallFunction|DetermineLaunchDirection _asfort_player_pawn)
        (if (not (GameplayTags|HasAnyMatchingGameplayTags (Variables|Default|GetCanLaunchTagContainer)))
          (Character|LaunchCharacter _asfort_player_pawn (Variables|Default|GetLaunchVelocity) true true)))
      (:CastFailed
        (bind _asfort_athena_vehicle (Utilities|Casting|CastToFortAthenaVehicle _array_element))
        (CallFunction|DetermineLaunchDirection _asfort_athena_vehicle)
        (Class|VehicleFunctionLibrary|LaunchVehicle _asfort_athena_vehicle (Variables|Default|GetLaunchVelocity))))))

(event OnComponentBeginOverlap(KillBox)
  (params (OverlappedComponent :type Primitive%20Component%20Object%20Reference) (OtherActor :type Actor%20Object%20Reference) (OtherComp :type Primitive%20Component%20Object%20Reference) (OtherBodyIndex :type Integer) (bFromSweep :type Boolean) (SweepResult :type Hit%20Result%20Structure%20%28by%20ref%29))
  (Variables|Default|SetSafteyLightOn true))

(event OnComponentEndOverlap(KillBox)
  (params (OverlappedComponent :type Primitive%20Component%20Object%20Reference) (OtherActor :type Actor%20Object%20Reference) (OtherComp :type Primitive%20Component%20Object%20Reference) (OtherBodyIndex :type Integer))
  (if (< (Utilities|Array|Length (Collision|GetOverlappingActors (Variables|Default|GetKillBox) "/Script/FortniteGame.FortPlayerPawn" :!class /Script/Engine.PrimitiveComponent)) 1)
    (Variables|Default|SetSafteyLightOn false)
    (elif (< (Utilities|Array|Length (Collision|GetOverlappingActors (Variables|Default|GetKillBox) "/Script/FortniteGame.FortAthenaVehicle" :!class /Script/Engine.PrimitiveComponent)) 1)
      (Variables|Default|SetSafteyLightOn false))))

(fn DetermineLaunchDirection
  (params (Actor :type Actor%20Object%20Reference))
  (Variables|Default|SetLaunchVelocity (* (* (* (+ (Transformation|GetActorRightVector) (Math|Vector|MakeVector 0.0 0.0 (Variables|Default|GetLaunchBoostInitialZ))) (Variables|Default|GetLaunchPower)) (Math|Vector|MakeVector (* (.x (* (+ (Transformation|GetActorRightVector) (Math|Vector|MakeVector 0.0 0.0 (Variables|Default|GetLaunchBoostInitialZ))) (Variables|Default|GetLaunchPower))) (Variables|Default|GetLaunchBoostX)) (* (.y (* (+ (Transformation|GetActorRightVector) (Math|Vector|MakeVector 0.0 0.0 (Variables|Default|GetLaunchBoostInitialZ))) (Variables|Default|GetLaunchPower))) (Variables|Default|GetLaunchBoostY)) (* (.z (* (+ (Transformation|GetActorRightVector) (Math|Vector|MakeVector 0.0 0.0 (Variables|Default|GetLaunchBoostInitialZ))) (Variables|Default|GetLaunchPower))) (Variables|Default|GetLaunchBoostZ)))) "1.000000,-1.000000,1.000000"))
  (return))

(fn MovePanelByPercentage
  (params (InputPin :type Float%20%28single-precision%29))
  (Transformation|SetRelativeRotation (Variables|Default|GetDoorPivot) (Math|Rotator|MakeRotator (Math|Float|Lerp (Variables|Default|GetRotationStartX) (Variables|Default|GetRotationEndX) InputPin) (.pitch (Variables|/Script/Engine.SceneComponent|GetRelativeRotation (Variables|Default|GetPanel))) (.yaw (Variables|/Script/Engine.SceneComponent|GetRelativeRotation (Variables|Default|GetPanel)))) false false))

(fn OnRep_SafteyLightOn
  (if (Variables|Default|GetSafteyLightOn)
    (Rendering|Material|SetMaterial (Variables|Default|GetSafteyLight) 0 "/DragonCartDoors/Assets/MI_FlatColorParent_Red.MI_FlatColorParent_Red")
    (return)
    (else
      (Rendering|Material|SetMaterial (Variables|Default|GetSafteyLight) 0 "/DragonCartDoors/Assets/MI_FlatColorParent_Iron.MI_FlatColorParent_Iron")
      (return))))

(fn ConstructionScript
  (|CallParentFunction))
